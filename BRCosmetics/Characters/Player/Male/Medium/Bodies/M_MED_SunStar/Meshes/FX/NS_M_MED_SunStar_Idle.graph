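# Idle Niagara effect for the SunStar character body that sheds dripping sprite droplets from head, arms, hands, pelvis and thighs.
# Rate and movement-based spawners, silenced while skydiving, split particles into eight partitions, each mapped to a bone with shaped random offsets.
# A curve-driven FallingLerp holds droplets on the skeletal mesh, then releases them to gravity and partial inherited movement.
# Sprites use system-computed alignment, time-of-day tint and visibility tags for hidden body parts.
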NiagaraSystem /BRCosmetics/Characters/Player/Male/Medium/Bodies/M_MED_SunStar/Meshes/FX/NS_M_MED_SunStar_Idle
dependencies {
 usedDataInterfaces [&NiagaraDataInterfaceCurve]
 usedDynamicInputs [
  &Add_Float
  &Lerp_Float
  &NormalizeVector
  /Niagara/DynamicInputs/Vectors/NegateVector
 ]
 usedModules [
  &self:SystemSpawnScript.SystemScriptSource.SystemScriptGraph.NiagaraNodeAssignment_1.SetVariables_#1
  &SystemState
  &InterpolateOverTime
  &self:SystemSpawnScript.SystemScriptSource.SystemScriptGraph.NiagaraNodeAssignment_0.SetVariables_#2
 ]
}
emitters {Skeletal_Mesh_Particle_Spawner {
  emitterUpdateScript [
   &EmitterState()
   &SpawnRate(
    Spawn Probability=&MakeCustomFloatFromBool(Bool=&Boolean_NotOperation(Bool=$User.IsSkydiving), True Float=0.33, False Float=0)
    SpawnRate=15
    Use Spawn Probability=true
   )
   &SpawnPerUnit(
    Module.Use Max Movement Threshold=true
    Module.Use Movement Tolerance=false
    Spawn Probability:NiagaraFloat=&MakeCustomFloatFromBool(Bool=&Boolean_NotOperation(Bool=$User.IsSkydiving), True Float=0.33, False Float=0)
    Spawn Spacing=60
    Use Spawn Probability=true
   )
   /SetVariables(Emitter.TimeofDay:NiagaraFloat=&Multiply_Float(A=Linked() {linkedVariable {name=NPC.FortniteNPC.TimeOfDay_0-1, type=NiagaraFloat}}, B=4)) {isSetParameters=true}
  ]
  particleSpawnScript [
   &PartitionParticles(Number Of Partitions=8)
   &InitializeParticle(
    Lifetime Min=1.333
    Lifetime Mode=::NewEnumerator1
    Sprite Size Mode=::NewEnumerator1
    Uniform Sprite Size Max=2
    Uniform Sprite Size Min=1.5
    UsePositionOffset=false
    Write Material Random=true
   )
   &ShapeLocation(
    Apply To Particle Position=::NewEnumerator1
    MaskSpawnGroup=false
    Override Seed=false
    Ring Radius=8
    Rotation Axis=<1 -0.3 0.02>
    Rotation Coordinate Space=::World
    Rotation Mode=::NewEnumerator1
    Shape Origin=&VectorFromFloat(Value=0)
    Shape Primitive=::NewEnumerator4
    UseOwnerScale=true
   )
   &ShapeLocation(
    Apply To Particle Position=::NewEnumerator1
    MaskSpawnGroup=false
    Non Uniform Scale=<1 0.9 1>
    Offset=<-5 0 0>
    Offset Coordinate Space=::World
    Offset Mode=::NewEnumerator0
    Override Seed=false
    Ring Radius=16
    Rotation Axis=<1 -0.1 0>
    Rotation Coordinate Space=::World
    Rotation Mode=::NewEnumerator1
    Shape Origin=&VectorFromFloat(Value=0)
    Shape Primitive=::NewEnumerator4
    U Distribution=0.6
    UseOwnerScale=true
   ) {module=ShapeLocation001}
   /SetVariables(Particles.SkeletalOffset:Vector3f=&self:ScratchDynamicInput(
     Index:NiagaraInt32=$Particles.PartitionParticles.Partition
     Head:Vector3f=&Multiply_Vector(
      A=&Lerp_Vector(A=<0 0 8>, B=<4 0 18>, Alpha=&RandomRangeFloat(Minimum=0, Maximum=1))
      B=&MakeCustomVectorFromBool(Bool=&RandomBool(), True Vector=<1 1 1>, False Vector=<1 1 -1>)
     )
     Lower Arm R:Vector3f=&RandomRangeVector(Minimum=<2 -0.5 -0.5>, Maximum=<-20 0.5 0.5>)
     Lower Arm L:Vector3f=&RandomRangeVector(Minimum=<-2 -0.5 -0.5>, Maximum=<20 0.5 0.5>)
     Hand R:Vector3f=&RandomRangeVector(Minimum=<0 0 -4>, Maximum=<-10 0 4>)
     Hand L:Vector3f=&RandomRangeVector(Minimum=<0 0 -4>, Maximum=<10 0 4>)
     Pelvis:Vector3f=$Particles.ShapeLocation001.ShapePosition
     Thigh R:Vector3f=&Add_Vector(A=<3 0 0>, B=$Particles.ShapeLocation.ShapePosition)
     Thigh L:Vector3f=&Add_Vector(A=<-3 0 0>, B=$Particles.ShapeLocation.ShapePosition)
    )) {isSetParameters=true}
   &AddVelocity(Rotation Coordinate Space=::World, Velocity=<0 0 -1>)
  ]
  particleUpdateScript [
   &ParticleState()
   /SetVariables(Particles.FallingLerp:NiagaraFloat=&FloatFromCurve(FloatCurve=DataInterface() {propertyValues {Curve {keys [<at 0 0 RCIM_Constant>,<at 0.66 1 RCIM_Constant>]}}})) {isSetParameters=true}
   &SkeletalMeshLocation(
    Bone / Socket Index=$Particles.PartitionParticles.Partition
    Bone Sampling Mode=::NewEnumerator4
    MaskSpawnGroup=false
    Sampled Position Offset=$Particles.SkeletalOffset
    Skeletal Mesh=$User.SkeletalMesh
    Use Sampled Mesh Attribute as Ribbon ID=false
    Write To Particle Position:NiagaraBool=&SetBoolByFloatComparison(A=$Particles.FallingLerp, B=0, Comparison Type=::NewEnumerator4)
   ) {module=SkeletalMeshLocation001}
   &InheritSourceMovement(
    Applied Force Scale=&VectorFromFloat(Value=-0.2)
    Applied Position Scale=&Lerp_Vector(A=<0 0 0>, B=<0.5 0.5 0.5>, Alpha=$Particles.FallingLerp)
    ApplyForce=false
    ApplyPosition=true
    ApplyPositionFalloff=false
   )
   &GravityForce(Gravity=&Multiply_VectorByFloat(Vector=<0 0 -980>, Float=$Particles.FallingLerp))
   &SolveForcesAndVelocity(Clamp Velocity=false, Limit Acceleration=false)
   &SpriteFacingAndAlignment(Alignment Coordinate Space=::World, Sprite Alignment=$System.AlignDroplets, WriteAlignment=true, WriteFacing=false)
   &ScaleSpriteSize(
    Scale Factor=&Multiply_Vector2D_ByFloat(
     Float=&FloatFromCurve(FloatCurve=DataInterface() {propertyValues {Curve {keys [
          {arriveTangent=2.1128, interpMode=RCIM_Cubic, leaveTangent=2.1128, tangentMode=RCTM_User, time=0, value=0}
          {interpMode=RCIM_Cubic, tangentMode=RCTM_User, time=0.66, value=1}
          {arriveTangent=-4.1172, interpMode=RCIM_Cubic, leaveTangent=-4.1172, tangentMode=RCTM_User, time=1, value=0}
         ]}}})
     Vector2D=&RandomRangeVector2D(Minimum=<1 2>, Maximum=<1 3>)
    )
    Scale Sprite Size Mode=::NewEnumerator2
   )
   &CameraOffset(Camera Offset Amount=5)
   &ScaleColor(
    Scale Mode=::NewEnumerator1
    Scale RGBA=&MakeVector4_FromLinearColor(LinearColor=&FN_Select_Color_from_Time_Of_Day(Morning=<rgba 1 0.5938 0.4583 1>, Day=<rgba 1 1 1 1>, Evening=<rgba 1 0.5938 0.4583 1>, Night=<rgba 0.125 0.3209 0.5 1>))
    ScaleRGBA=true
   )
   /SetVariables(Particles.VisibilityTag=&self:Part_Hidden_CHeck(
     is Head Hidden:NiagaraBool=$User.bisHeadHidden
     Partition:NiagaraInt32=$Particles.PartitionParticles.Partition
     is Hand Hidden:NiagaraBool=$User.bisHandHidden
     is Foot Hidden:NiagaraBool=$User.bisFootHidden
    )) {isSetParameters=true}
  ]
  rendererClasses [&NiagaraSpriteRendererProperties]
  renderers [{rendererClass=&NiagaraSpriteRendererProperties}]
 }}
summary {
 systemName=NS_M_MED_SunStar_Idle
 emitters [{emitterName=Skeletal_Mesh_Particle_Spawner, rendererClasses [&NiagaraSpriteRendererProperties]}]
 userVariables [
  User.bisFootHidden {type=NiagaraBool}
  User.bisFrontEndPreview {type=NiagaraBool}
  User.bisHandHidden {type=NiagaraBool}
  User.bisHeadHidden {type=NiagaraBool}
  User.Initial Rotation {type=NiagaraFloat, defaultValue=0, description=""}
  User.Initial Rotation Range {type=NiagaraFloat, defaultValue=0, description=""}
  User.IsSkydiving {type=NiagaraBool}
  User.SkeletalMesh {
   type=NiagaraDataInterfaceSkeletalMesh
   defaultValue {
    struct=NiagaraExt_VariableValue_DataInterface
    value {dataInterface=&self:NiagaraDataInterfaceSkeletalMesh_0}
   }
  }
 ]
}
