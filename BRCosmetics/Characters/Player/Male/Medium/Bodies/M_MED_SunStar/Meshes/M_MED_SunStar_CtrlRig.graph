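# Control Rig for the M_MED_SunStar male medium body cosmetic that procedurally adjusts the shorts bones as the knees bend.
# On forward solve, each leg reads calf rotation relative to thigh, converts it to Euler and takes the absolute Z angle.
# A clamped remap turns that angle into the weight of a Modify Transforms step that offsets that side's dyn shorts bones, including back and inner.
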ControlRigBlueprint /BRCosmetics/Characters/Player/Male/Medium/Bodies/M_MED_SunStar/Meshes/M_MED_SunStar_CtrlRig
graphs [
 RigVMFunctionLibrary {links [], nodes []}
 RigVMModel {nodes [
   RigUnit_BeginExecution = &RigUnit_BeginExecution()
   n1 = &RigVMFunction_MathQuaternionToEuler(
    RotationOrder=ZYX
    Value=&RigUnit_GetRelativeTransformForItem(Child=(Type=Bone,Name="calf_l"), Parent=(Type=Bone,Name="thigh_l")).RelativeTransform.Rotation
   )
   &RigUnit_ModifyTransforms(
    ItemToModify="((Item=(Type=Bone,Name=\"dyn_L_shorts_bk\"),Transform=(Rotation=(X=0,Y=0,Z=0,W=1),Translation=(X=-3,Y=1,Z=0),Scale3D=(X=1,Y=1,Z=1))),(Item=(Type=Bone,Name=\"dyn_L_shorts_in\"),Transform=(Rotation=(X=0,Y=0,Z=0,W=1),Translation=(X=0,Y=0,Z=0),Scale3D=(X=1,Y=1,Z=1 ...(truncated)"
    Weight=&RigVMFunction_MathFloatRemap(
     SourceMaximum=75
     SourceMinimum=10
     TargetMinimum=0
     bClamp=true
     Value=/Script/RigVM.RigVMFunction_MathFloatAbs(Value=@n1.Result.Z) {is=struct RigVMUnitNode, template="Absolute::Execute(in Value,out Result)"}.Result
    ).Result
    ItemToModify.0.Transform.Rotation=&RigVMFunction_MathQuaternionFromEuler(Euler=(X=0,Y=0,Z=-100)).Result
    ItemToModify.1.Transform.Rotation=&RigVMFunction_MathQuaternionFromEuler(Euler=(X=0,Y=0,Z=-90)).Result
    ItemToModify.2.Transform.Rotation=&RigVMFunction_MathQuaternionFromEuler(Euler=(X=0,Y=0,Z=-90)).Result
    ExecutePin=@RigUnit_BeginExecution.ExecutePin
   )
   n2 = &RigVMFunction_MathQuaternionToEuler(
    RotationOrder=ZYX
    Value=&RigUnit_GetRelativeTransformForItem(Child=(Type=Bone,Name="calf_r"), Parent=(Type=Bone,Name="thigh_r")).RelativeTransform.Rotation
   )
   > &RigUnit_ModifyTransforms(
    ItemToModify="((Item=(Type=Bone,Name=\"dyn_R_shorts_bk\"),Transform=(Rotation=(X=0,Y=0,Z=0,W=1),Translation=(X=0,Y=0,Z=0),Scale3D=(X=1,Y=1,Z=1))),(Item=(Type=Bone,Name=\"dyn_R_shorts_in\"),Transform=(Rotation=(X=0,Y=0,Z=0,W=1),Translation=(X=0,Y=0,Z=0),Scale3D=(X=1,Y=1,Z=1) ...(truncated)"
    Weight=&RigVMFunction_MathFloatRemap(
     SourceMaximum=75
     SourceMinimum=10
     TargetMinimum=0
     bClamp=true
     Value=&RigVMFunction_MathFloatAbs(Value=@n2.Result.Z) {template="Absolute::Execute(in Value,out Result)"}.Result
    ).Result
    ItemToModify.2.Transform.Rotation=&RigVMFunction_MathQuaternionFromEuler(Euler=(X=0,Y=0,Z=-90)).Result
    ItemToModify.1.Transform.Rotation=&RigVMFunction_MathQuaternionFromEuler(Euler=(X=0,Y=0,Z=-90)).Result
    ItemToModify.0.Transform.Rotation=&RigVMFunction_MathQuaternionFromEuler(Euler=(X=0,Y=0,Z=-130)).Result
   )
  ]}
]
hierarchy [
 root {loc=<0 0 0>, rot=<0 0 0 1>}
 pelvis {parent=root, loc=<0 -0.7071 93.1191>, rot=<0.0127 0.707 -0.0127 -0.707>}
 spine_01 {parent=pelvis, loc=<6.7494 -0.6388 0>, rot=<0 0 0.0503 -0.9987>}
 spine_02 {parent=spine_01, loc=<7.2493 -0.0308 0>, rot=<0 0 0.0041 -1>}
 spine_03 {parent=spine_02, loc=<6.0084 -0.0167 0>, rot=<0 0 0.0027 -1>}
 spine_04 {parent=spine_03, loc=<8.0999 -0.0224 0>, rot=<0 0 0.0026 -1>}
 spine_05 {parent=spine_04, loc=<10.0119 -0.0184 0>, rot=<0 0 -0.0016 -1>}
 clavicle_l {parent=spine_05, loc=<4.3575 -0.1466 -2.3223>, rot=<0.0476 -0.7289 -0.0102 -0.6829>}
 upperarm_l {parent=clavicle_l, loc=<15.5111 0 0>, rot=<0.0221 -0.3319 -0.0095 -0.943>}
 lowerarm_l {parent=upperarm_l, loc=<26.647 0 0>, rot=<0 0 -0.2447 -0.9696>}
 hand_l {parent=lowerarm_l, loc=<22.344 0 0>, rot=<-0.6794 -0.0546 0.1067 -0.7239>}
 weapon_l {parent=hand_l, loc=<-0.1974 -3.4111 0.3811>, rot=<0 0 0 1>}
 thumb_01_l {parent=hand_l, loc=<1.3118 -0.3191 -2.459>, rot=<0.6546 -0.4042 0.0664 -0.6354>}
 thumb_02_l {parent=thumb_01_l, loc=<5.9336 -0.0448 0.0401>, rot=<-0.0088 0.005 0.2979 -0.9545>}
 thumb_03_l {parent=thumb_02_l, loc=<2.2223 0.3682 0>, rot=<0 0 0.0067 -1>}
 index_metacarpal_l {parent=hand_l, loc=<1.6535 0.0183 -2.96>, rot=<0.0158 -0.0124 -0.0173 -0.9996>}
 index_01_l {parent=index_metacarpal_l, loc=<8.3008 0.0357 -0.0524>, rot=<-0.0044 0.0485 0.0924 -0.9945>}
 index_02_l {parent=index_01_l, loc=<3.3587 -0.3004 0>, rot=<0 0 0.3225 -0.9466>}
 index_03_l {parent=index_02_l, loc=<2.5578 -0.123 0>, rot=<0 0 0.2559 -0.9667>}
 middle_metacarpal_l {parent=hand_l, loc=<2.1856 -0.1725 -0.5984>, rot=<0.006 0.0168 -0.0329 -0.9993>}
 middle_01_l {parent=middle_metacarpal_l, loc=<7.3275 -0.2447 0.0129>, rot=<-0.0208 0.0042 0.0858 -0.9961>}
 middle_02_l {parent=middle_01_l, loc=<4.016 0.1415 0>, rot=<0 0 0.3519 -0.936>}
 middle_03_l {parent=middle_02_l, loc=<2.6103 0.0517 0>, rot=<0 0 0.1707 -0.9853>}
 ring_metacarpal_l {parent=hand_l, loc=<2.2074 -0.3625 1.3638>, rot=<-0.0284 0.0722 -0.0458 -0.9959>}
 ring_01_l {parent=ring_metacarpal_l, loc=<7.0276 -0.6134 -0.2041>, rot=<-0.0484 -0.0225 0.1894 -0.9805>}
 ring_02_l {parent=ring_01_l, loc=<3.6174 0.4647 0>, rot=<0 0 0.2806 -0.9598>}
 ring_03_l {parent=ring_02_l, loc=<2.4152 0.114 0>, rot=<0 0 0.1189 -0.9929>}
 pinky_metacarpal_l {parent=hand_l, loc=<1.7616 -0.9753 3.3814>, rot=<-0.0397 0.0814 -0.069 -0.9935>}
 pinky_01_l {parent=pinky_metacarpal_l, loc=<7.1709 -0.4831 -0.1094>, rot=<-0.0676 -0.0132 0.2236 -0.9723>}
 pinky_02_l {parent=pinky_01_l, loc=<2.8274 0.0893 0>, rot=<0 0 0.2854 -0.9584>}
 pinky_03_l {parent=pinky_02_l, loc=<2.2046 0.0307 0>, rot=<0 0 0.1445 -0.9895>}
 hand_attach_l {parent=hand_l, loc=<8.51 -3 0>, rot=<0.7071 0.7071 0 0>}
 deform_wrist_dn_base_l {parent=hand_l, loc=<0 -2.2342 0>, rot=<0 0 0 -1>}
 deform_wrist_dn_l {parent=deform_wrist_dn_base_l, loc=<0 -0.085 0>, rot=<0 0 0 1>, scale=<0.9929 1 1>}
 deform_wrist_up_base_l {parent=hand_l, loc=<0 2.2682 0>, rot=<0 0 0 -1>}
 deform_wrist_up_l {parent=deform_wrist_up_base_l, loc=<0 0 0>, rot=<0 0 0 1>}
 lowerarm_twist_02_l {parent=lowerarm_l, loc=<9.6199 0 0>, rot=<0 0 0 1>}
 lowerarm_twist_01_l {parent=lowerarm_l, loc=<17.8549 0 0>, rot=<0 0 0 1>}
 deform_elbow_in_base_l {parent=lowerarm_l, loc=<0.2201 5.3074 0.8925>, rot=<0.0388 -0.0107 -0.0004 -0.9992>}
 deform_elbow_in_l {parent=deform_elbow_in_base_l, loc=<0 0 0>, rot=<0 0 0 1>}
 deform_elbow_out_base_l {parent=lowerarm_l, loc=<0.0067 -3.9879 0.7908>, rot=<0.0388 -0.0107 -0.0004 -0.9992>}
 deform_elbow_out_l {parent=deform_elbow_out_base_l, loc=<0 0 0>, rot=<0 0 0 1>}
 dyn_L_bracelet_jnt {parent=lowerarm_l, loc=<20.8608 -0.3836 0.5543>, rot=<-0.7363 -0.0053 0.0688 -0.6731>}
 master_L_armfur_jnt {parent=lowerarm_l, loc=<0 0 0>, rot=<0 0 0 1>}
 dyn_L_armfur_A_jnt {parent=master_L_armfur_jnt, loc=<-2.3928 2.3324 -4.8053>, rot=<-0.736 -0.2004 0.305 -0.5701>}
 dyn_L_armfur_B_jnt {parent=master_L_armfur_jnt, loc=<6.6941 1.6801 4.7476>, rot=<0.7925 0.0933 0.1085 -0.5928>}
 dyn_L_armfur_C_jnt {parent=master_L_armfur_jnt, loc=<6.1143 -4.7599 2.3045>, rot=<0.2102 0.0199 0.1625 -0.9638>}
 dyn_L_armfur_D_jnt {parent=master_L_armfur_jnt, loc=<4.2041 -2.7716 -3.9452>, rot=<-0.5338 0.0584 0.0311 -0.843>}
 dyn_L_armfur_E_jnt {parent=master_L_armfur_jnt, loc=<4.921 3.5605 -2.8699>, rot=<0.9724 0.0056 -0.0264 0.2319>}
 dyn_L_armfur_F_jnt {parent=master_L_armfur_jnt, loc=<-0.6604 -0.3496 4.5145>, rot=<0.6219 0.1284 0.2085 -0.7439>}
 dyn_L_armfur_G_jnt {parent=master_L_armfur_jnt, loc=<-4.4239 -2.3784 0.148>, rot=<0.1168 0.0103 0.459 -0.8806>}
 upperarm_twist_01_l {parent=upperarm_l, loc=<0.0041 0 0>, rot=<0 0 0 1>}
 deform_upperarm_bk_base_l {parent=upperarm_twist_01_l, loc=<1.1164 -7.2153 -1.9401>, rot=<-0.001 0.009 0 -1>}
 deform_upperarm_bk_l {parent=deform_upperarm_bk_base_l, loc=<0 0 0>, rot=<0 0 0 1>}
 deform_upperarm_fr_base_l {parent=upperarm_twist_01_l, loc=<-0.3194 7.0563 -2.7568>, rot=<-0.001 0.009 0 -1>}
 deform_upperarm_fr_l {parent=deform_upperarm_fr_base_l, loc=<0 0 0>, rot=<0 0 0 1>}
 deform_upperarm_up_base_l {parent=upperarm_twist_01_l, loc=<-3.2811 -1.1459 3.2702>, rot=<-0.001 0.009 0 -1>}
 deform_upperarm_up_l {parent=deform_upperarm_up_base_l, loc=<0 0 0>, rot=<0 0 0 1>}
 upperarm_twist_02_l {parent=upperarm_l, loc=<13.7941 0 0>, rot=<0 0 0 1>}
 neck_01 {parent=spine_05, loc=<11.3138 0.1529 0>, rot=<0 0 -0.2003 -0.9797>}
 neck_02 {parent=neck_01, loc=<4.5627 0 0>, rot=<0 0 0.0504 -0.9987>}
 head {parent=neck_02, loc=<5.3914 0 0>, rot=<0 0 0.115 -0.9934>}
 clavicle_r {parent=spine_05, loc=<4.3575 -0.1466 2.3223>, rot=<0.7289 0.0476 -0.6829 0.0102>}
 upperarm_r {parent=clavicle_r, loc=<-15.5111 0 0>, rot=<0.0221 -0.3319 -0.0095 -0.943>}
 lowerarm_r {parent=upperarm_r, loc=<-26.647 0 0>, rot=<0 0 -0.2447 -0.9696>}
 hand_r {parent=lowerarm_r, loc=<-22.344 0 0>, rot=<-0.6794 -0.0546 0.1067 -0.7239>}
 thumb_01_r {parent=hand_r, loc=<-1.3118 0.3191 2.459>, rot=<0.6546 -0.4042 0.0664 -0.6354>}
 thumb_02_r {parent=thumb_01_r, loc=<-5.9336 0.0448 -0.0401>, rot=<-0.0088 0.005 0.2979 -0.9545>}
 thumb_03_r {parent=thumb_02_r, loc=<-2.2223 -0.3682 0>, rot=<0 0 0.0067 -1>}
 index_metacarpal_r {parent=hand_r, loc=<-1.6535 -0.0183 2.96>, rot=<0.0158 -0.0124 -0.0173 -0.9996>}
 index_01_r {parent=index_metacarpal_r, loc=<-8.3008 -0.0357 0.0524>, rot=<-0.0044 0.0485 0.0924 -0.9945>}
 index_02_r {parent=index_01_r, loc=<-3.3587 0.3004 0>, rot=<0 0 0.3225 -0.9466>}
 index_03_r {parent=index_02_r, loc=<-2.5578 0.123 0>, rot=<0 0 0.2559 -0.9667>}
 middle_metacarpal_r {parent=hand_r, loc=<-2.1856 0.1725 0.5984>, rot=<0.006 0.0168 -0.0329 -0.9993>}
 middle_01_r {parent=middle_metacarpal_r, loc=<-7.3275 0.2447 -0.0129>, rot=<-0.0208 0.0042 0.0858 -0.9961>}
 middle_02_r {parent=middle_01_r, loc=<-4.016 -0.1415 0>, rot=<0 0 0.3519 -0.936>}
 middle_03_r {parent=middle_02_r, loc=<-2.6103 -0.0517 0>, rot=<0 0 0.1707 -0.9853>}
 ring_metacarpal_r {parent=hand_r, loc=<-2.2074 0.3625 -1.3638>, rot=<-0.0284 0.0722 -0.0458 -0.9959>}
 ring_01_r {parent=ring_metacarpal_r, loc=<-7.0276 0.6134 0.2041>, rot=<-0.0484 -0.0225 0.1894 -0.9805>}
 ring_02_r {parent=ring_01_r, loc=<-3.6174 -0.4647 0>, rot=<0 0 0.2806 -0.9598>}
 ring_03_r {parent=ring_02_r, loc=<-2.4152 -0.114 0>, rot=<0 0 0.1189 -0.9929>}
 pinky_metacarpal_r {parent=hand_r, loc=<-1.7616 0.9753 -3.3814>, rot=<-0.0397 0.0814 -0.069 -0.9935>}
 pinky_01_r {parent=pinky_metacarpal_r, loc=<-7.1709 0.4831 0.1094>, rot=<-0.0676 -0.0132 0.2236 -0.9723>}
 pinky_02_r {parent=pinky_01_r, loc=<-2.8274 -0.0893 0>, rot=<0 0 0.2854 -0.9584>}
 pinky_03_r {parent=pinky_02_r, loc=<-2.2046 -0.0307 0>, rot=<0 0 0.1445 -0.9895>}
 weapon_r {parent=hand_r, loc=<0.1974 3.4111 -0.3811>, rot=<0 0 0 1>}
 hand_attach_r {parent=hand_r, loc=<-8.51 2.9999 0>, rot=<0 0 -0.7071 -0.7071>}
 deform_wrist_dn_base_r {parent=hand_r, loc=<-0.0004 2.2345 -0.0001>, rot=<0 0 0 -1>}
 deform_wrist_dn_r {parent=deform_wrist_dn_base_r, loc=<0 0.085 0>, rot=<0 0 0 1>, scale=<0.9929 1 1>}
 deform_wrist_up_base_r {parent=hand_r, loc=<0.0003 -2.2684 0.0001>, rot=<0 0 0 -1>}
 deform_wrist_up_r {parent=deform_wrist_up_base_r, loc=<0 0 0>, rot=<0 0 0 1>}
 lowerarm_twist_02_r {parent=lowerarm_r, loc=<-9.6199 0 0>, rot=<0 0 0 -1>}
 lowerarm_twist_01_r {parent=lowerarm_r, loc=<-17.8549 0 0>, rot=<0 0 0 -1>}
 deform_elbow_in_base_r {parent=lowerarm_r, loc=<-0.22 -5.3074 -0.8926>, rot=<0.0388 -0.0107 -0.0004 -0.9992>}
 deform_elbow_in_r {parent=deform_elbow_in_base_r, loc=<-0.0001 0 0.0002>, rot=<0 0 0 1>}
 deform_elbow_out_base_r {parent=lowerarm_r, loc=<-0.0067 3.9879 -0.7908>, rot=<0.0388 -0.0107 -0.0004 -0.9992>}
 deform_elbow_out_r {parent=deform_elbow_out_base_r, loc=<0 0 0>, rot=<0 0 0 1>}
 dyn_R_bracelet_jnt {parent=lowerarm_r, loc=<-20.8606 0.3835 -0.5544>, rot=<-0.7363 -0.0053 0.0688 -0.6731>}
 master_R_armfur_jnt {parent=lowerarm_r, loc=<-0.0002 0.0001 0.0002>, rot=<0 0 0 -1>}
 dyn_R_armfur_A_jnt {parent=master_R_armfur_jnt, loc=<2.3928 -2.3324 4.8054>, rot=<-0.736 -0.2004 0.305 -0.5701>}
 dyn_R_armfur_B_jnt {parent=master_R_armfur_jnt, loc=<-8.2417 -4.6575 -2.6167>, rot=<0.9843 0.0493 0.0947 -0.1405>}
 dyn_R_armfur_C_jnt {parent=master_R_armfur_jnt, loc=<-6.1142 4.7599 -2.3046>, rot=<0.2102 0.0199 0.1625 -0.9638>}
 dyn_R_armfur_D_jnt {parent=master_R_armfur_jnt, loc=<-4.204 2.7716 3.9452>, rot=<-0.5338 0.0584 0.0311 -0.843>}
 dyn_R_armfur_E_jnt {parent=master_R_armfur_jnt, loc=<-4.9208 -3.5606 2.8697>, rot=<0.9724 0.0056 -0.0264 0.2319>}
 dyn_R_armfur_F_jnt {parent=master_R_armfur_jnt, loc=<2.5033 0.0421 -3.6094>, rot=<0.5985 0.2122 0.3082 -0.7084>}
 dyn_R_armfur_G_jnt {parent=master_R_armfur_jnt, loc=<3.523 3.4355 0.6337>, rot=<0.0108 -0.0077 0.4034 -0.9149>}
 upperarm_twist_01_r {parent=upperarm_r, loc=<-0.0041 0 0>, rot=<0 0 0 -1>}
 deform_upperarm_bk_base_r {parent=upperarm_twist_01_r, loc=<-1.1166 7.2153 1.9404>, rot=<-0.001 0.009 0 -1>}
 deform_upperarm_bk_r {parent=deform_upperarm_bk_base_r, loc=<0 0 0>, rot=<0 0 0 1>}
 deform_upperarm_fr_base_r {parent=upperarm_twist_01_r, loc=<0.3193 -7.0563 2.7569>, rot=<-0.001 0.009 0 -1>}
 deform_upperarm_fr_r {parent=deform_upperarm_fr_base_r, loc=<0 0 0>, rot=<0 0 0 1>}
 deform_upperarm_up_base_r {parent=upperarm_twist_01_r, loc=<3.2808 1.1459 -3.27>, rot=<-0.001 0.009 0 -1>}
 deform_upperarm_up_r {parent=deform_upperarm_up_base_r, loc=<0 0 0>, rot=<0 0 0 -1>}
 upperarm_twist_02_r {parent=upperarm_r, loc=<-13.7941 0 0>, rot=<0 0 0 -1>}
 master_star_medallion_jnt {parent=spine_05, loc=<7.0934 4.3491 -0.1311>, rot=<0.0401 -0.9992 0 0>}
 dyn_star_medallion_jnt {parent=master_star_medallion_jnt, loc=<0 2.0698 0>, rot=<0 0 -0.2333 -0.9724>}
 deform_pec_aim_base_l {parent=spine_04, loc=<7.8667 12.6198 -2.4123>, rot=<0.0103 -0.4602 0.3978 -0.7937>}
 deform_pec_aim_l {parent=deform_pec_aim_base_l, loc=<0 0 0>, rot=<0 0 0 1>}
 deform_pec_base_l {parent=deform_pec_aim_l, loc=<10.766 0 0>, rot=<0 0 0 1>}
 deform_pec_l {parent=deform_pec_base_l, loc=<0.055 0 0>, rot=<0 0 0 1>}
 deform_pec_aim_base_r {parent=spine_04, loc=<7.8662 12.6197 2.4123>, rot=<-0.4602 -0.0103 0.7937 0.3978>}
 deform_pec_aim_r {parent=deform_pec_aim_base_r, loc=<0 0 0>, rot=<0 0 0 1>}
 deform_pec_base_r {parent=deform_pec_aim_r, loc=<-10.7662 -0.0006 -0.0007>, rot=<0 0 0 1>}
 deform_pec_r {parent=deform_pec_base_r, loc=<0 0 0>, rot=<0 0 0 1>}
 thigh_l {parent=pelvis, loc=<-2.5693 -0.0109 -7.7501>, rot=<0.0013 0.0462 0.0155 -0.9988>}
 calf_l {parent=thigh_l, loc=<-39.039 0 0>, rot=<0 0 -0.0378 -0.9993>}
 foot_l {parent=calf_l, loc=<-42.5546 0 0>, rot=<0.0427 -0.0505 0.0328 -0.9973>}
 ball_l {parent=foot_l, loc=<-6.1599 13.8264 0.6806>, rot=<0.0001 -0.0001 -0.7063 -0.7079>}
 calf_twist_02_l {parent=calf_l, loc=<-18.8064 -0.0967 0.0397>, rot=<0.0237 0.0095 -0.0035 -0.9997>}
 calf_twist_01_l {parent=calf_l, loc=<-35.813 -0.0967 0.0397>, rot=<0.0237 0.0095 -0.0035 -0.9997>}
 deform_knee_in_base_l {parent=calf_l, loc=<1.311 -5.2925 -0.9166>, rot=<0.0007 -0.0051 -0.0013 -1>}
 deform_knee_in_l {parent=deform_knee_in_base_l, loc=<0 0 0>, rot=<0 0 0 1>}
 deform_knee_out_base_l {parent=calf_l, loc=<-0.0102 4 -0.866>, rot=<0.0007 -0.0051 -0.0013 -1>}
 deform_knee_out_l {parent=deform_knee_out_base_l, loc=<0 0 0>, rot=<0 0 0 1>}
 thigh_twist_01_l {parent=thigh_l, loc=<-21.629 -0.0919 -0.0002>, rot=<0.0013 0 0 -1>}
 master_L_dyn_shorts_main {parent=thigh_twist_01_l, loc=<-10.1679 -0.4394 -0.2741>, rot=<-0.0023 -0.9989 -0.0017 0.0462>}
 dyn_L_shorts_main {parent=master_L_dyn_shorts_main, loc=<2.9776 0 0.0001>, rot=<0 0 0 1>}
 dyn_L_shorts_out {parent=dyn_L_shorts_main, loc=<0.1596 0.0637 8.7231>, rot=<0 0 0 1>}
 dyn_L_shorts_out_02 {parent=dyn_L_shorts_out, loc=<7.6722 0.0733 -0.2988>, rot=<0 0 0 1>}
 dyn_L_shorts_in {parent=dyn_L_shorts_main, loc=<1.5204 1.1393 -7.0205>, rot=<0 0 0 1>}
 dyn_L_shorts_in_02 {parent=dyn_L_shorts_in, loc=<7.9025 -0.1774 0.0411>, rot=<0 0 0 1>}
 dyn_L_shorts_bk {parent=dyn_L_shorts_main, loc=<1.1015 -7.0243 0.1356>, rot=<0 0 0 1>}
 dyn_L_shorts_bk_02 {parent=dyn_L_shorts_bk, loc=<7.8126 -0.2518 1.5412>, rot=<0 0 0 1>}
 dyn_L_shorts_fr {parent=dyn_L_shorts_main, loc=<1.6211 8.7787 0.6086>, rot=<0 0 0 1>}
 dyn_L_shorts_fr_02 {parent=dyn_L_shorts_fr, loc=<7.6827 -0.407 0.0177>, rot=<0 0 0 1>}
 deform_groin_base_l {parent=thigh_l, loc=<-2.1519 8.1149 -1.7885>, rot=<-0.0002 0.011 0 -0.9999>}
 deform_groin_l {parent=deform_groin_base_l, loc=<0 0 0>, rot=<0 0 0 1>}
 thigh_r {parent=pelvis, loc=<-2.5693 -0.0109 7.7501>, rot=<0.0462 -0.0013 0.9988 0.0155>}
 calf_r {parent=thigh_r, loc=<39.0389 0 0>, rot=<0 0 -0.0378 -0.9993>}
 foot_r {parent=calf_r, loc=<42.5546 0 0>, rot=<0.0427 -0.0505 0.0328 -0.9973>}
 ball_r {parent=foot_r, loc=<6.1599 -13.8264 -0.6806>, rot=<0.0001 -0.0001 -0.7063 -0.7079>}
 calf_twist_02_r {parent=calf_r, loc=<18.8064 0.0967 -0.0397>, rot=<0.0237 0.0095 -0.0035 -0.9997>}
 calf_twist_01_r {parent=calf_r, loc=<35.813 0.0967 -0.0397>, rot=<0.0237 0.0095 -0.0035 -0.9997>}
 deform_knee_in_base_r {parent=calf_r, loc=<-1.311 5.2925 0.9166>, rot=<0.0007 -0.0051 -0.0013 -1>}
 deform_knee_in_r {parent=deform_knee_in_base_r, loc=<0 0 0>, rot=<0 0 0 1>}
 deform_knee_out_base_r {parent=calf_r, loc=<0.0103 -4 0.866>, rot=<0.0007 -0.0051 -0.0013 -1>}
 deform_knee_out_r {parent=deform_knee_out_base_r, loc=<0 0 0>, rot=<0 0 0 1>}
 thigh_twist_01_r {parent=thigh_r, loc=<21.629 0.0919 0.0002>, rot=<0.0013 0 0 -1>}
 master_R_dyn_shorts_main {parent=thigh_twist_01_r, loc=<10.0921 0.4397 0.2811>, rot=<-0.0023 -0.9989 -0.0017 0.0462>}
 dyn_R_shorts_main {parent=master_R_dyn_shorts_main, loc=<-2.9287 0 -0.0001>, rot=<0 0 0 1>}
 dyn_R_shorts_out {parent=dyn_R_shorts_main, loc=<-0.2047 -0.0637 -8.7232>, rot=<0 0 0 1>}
 dyn_R_shorts_out_02 {parent=dyn_R_shorts_out, loc=<-8.0576 -0.0733 0.2988>, rot=<0 0 0 1>}
 dyn_R_shorts_in {parent=dyn_R_shorts_main, loc=<-1.5654 -1.1393 7.0205>, rot=<0 0 0 1>}
 dyn_R_shorts_in_02 {parent=dyn_R_shorts_in, loc=<-8.2879 0.1774 -0.0411>, rot=<0 0 0 1>}
 dyn_R_shorts_bk {parent=dyn_R_shorts_main, loc=<-1.1466 7.0243 -0.1356>, rot=<0 0 0 1>}
 dyn_R_shorts_bk_02 {parent=dyn_R_shorts_bk, loc=<-8.1979 0.2518 -1.5412>, rot=<0 0 0 1>}
 dyn_R_shorts_fr {parent=dyn_R_shorts_main, loc=<-1.6662 -8.7787 -0.6086>, rot=<0 0 0 1>}
 dyn_R_shorts_fr_02 {parent=dyn_R_shorts_fr, loc=<-8.068 0.407 -0.0177>, rot=<0 0 0 1>}
 deform_groin_base_r {parent=thigh_r, loc=<2.1519 -8.1149 1.7885>, rot=<-0.0002 0.011 0 -0.9999>}
 deform_groin_r {parent=deform_groin_base_r, loc=<0 0 0>, rot=<0 0 0 1>}
 deform_glute_base_l {parent=pelvis, loc=<-8.8064 -8.8222 -8.4287>, rot=<0.0275 0.0501 -0.4472 -0.8926>}
 deform_glute_l {parent=deform_glute_base_l, loc=<-0.0067 -0.0167 0>, rot=<0 0 0 1>}
 deform_glute_base_r {parent=pelvis, loc=<-8.8062 -8.8222 8.4287>, rot=<0.0501 -0.0275 0.8926 -0.4472>}
 deform_glute_r {parent=deform_glute_base_r, loc=<0.0067 0.0167 0>, rot=<0 0 0 1>}
 attach {parent=root, loc=<0 0 0>, rot=<0 0 0 1>}
 ik_foot_root {parent=root, loc=<0 0 0>, rot=<0 0 0 1>}
 ik_foot_l {parent=ik_foot_root, loc=<15.2608 -4.4176 9.4432>, rot=<0.0357 0.7045 0.0284 -0.7082>}
 ik_foot_r {parent=ik_foot_root, loc=<-15.2608 -4.4176 9.4432>, rot=<0.7082 0.0284 -0.7045 0.0357>}
 ik_hand_root {parent=root, loc=<0 0 0>, rot=<0 0 0 1>}
 ik_hand_gun {parent=ik_hand_root, loc=<-51.9727 4.0883 102.6935>, rot=<0.6084 0.1793 0.4613 -0.6205>}
 ik_hand_r {parent=ik_hand_gun, loc=<0 0 0>, rot=<0 0 0 1>}
 ik_hand_l {parent=ik_hand_gun, loc=<53.0285 82.1781 35.2051>, rot=<-0.5102 -0.7906 -0.3387 0>}
]
