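# Physics asset for the SunStar medium male body cosmetic, covering its dynamic arm fur and star medallion joints.
# Seven fur capsules per arm attach with tight swing and twist limits to small master bodies on each forearm.
# The medallion hangs from a master body on spine_05, swinging widely on one axis only.
# Forearm and spine capsules supply body volumes, linked by rotationally free constraints.
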PhysicsAsset /BRCosmetics/Characters/Player/Male/Medium/Bodies/M_MED_SunStar/Meshes/M_MED_SunStar_Physics
bodies {
 dyn_L_armfur_A_jnt [{shapeType=Capsule, length=5, radius=2, shapeName=dyn_L_armfur_A_jnt_capsule, center=<3.9211 0 0>, rotation=<rot 90 0 0>}]
 dyn_L_armfur_B_jnt [{shapeType=Capsule, length=5, radius=2.75, shapeName=dyn_L_armfur_B_jnt_capsule, center=<3.9437 0 0>, rotation=<rot 90 0 0>}]
 dyn_L_armfur_C_jnt [{shapeType=Capsule, length=5, radius=2.75, shapeName=dyn_L_armfur_C_jnt_capsule, center=<3.8844 0 0>, rotation=<rot 90 0 0>}]
 dyn_L_armfur_D_jnt [{shapeType=Capsule, length=5, radius=2, shapeName=dyn_L_armfur_D_jnt_capsule, center=<3.9983 0 0>, rotation=<rot 90 0 0>}]
 dyn_L_armfur_E_jnt [{shapeType=Capsule, length=5, radius=2, shapeName=dyn_L_armfur_E_jnt_capsule, center=<4.2043 0 0>, rotation=<rot 90 0 0>}]
 dyn_L_armfur_F_jnt [{shapeType=Capsule, length=5, radius=2, shapeName=dyn_L_armfur_F_jnt_capsule, center=<4.1262 0 0>, rotation=<rot 90 0 0>}]
 dyn_L_armfur_G_jnt [{shapeType=Capsule, length=5, radius=2, shapeName=dyn_L_armfur_G_jnt_capsule, center=<4.504 0 0>, rotation=<rot 90 0 0>}]
 dyn_R_armfur_A_jnt [{shapeType=Capsule, length=5, radius=2, shapeName=dyn_R_armfur_A_jnt_capsule, center=<-4.1654 0 0>, rotation=<rot 90 0 0>}]
 dyn_R_armfur_B_jnt [{
   shapeType=Capsule
   length=5
   radius=3.5
   shapeName=dyn_R_armfur_B_jnt_capsule
   center=<-3.7923 0 -0.5154>
   rotation=<rot 80 -179.9997 170.0004>
  }]
 dyn_R_armfur_C_jnt [{shapeType=Capsule, length=5, radius=2, shapeName=dyn_R_armfur_C_jnt_capsule, center=<-4.3412 0 0>, rotation=<rot 90 0 0>}]
 dyn_R_armfur_D_jnt [{shapeType=Capsule, length=5, radius=2, shapeName=dyn_R_armfur_D_jnt_capsule, center=<-4.4253 0 0>, rotation=<rot 90 0 0>}]
 dyn_R_armfur_E_jnt [{shapeType=Capsule, length=5, radius=2, shapeName=dyn_R_armfur_E_jnt_capsule, center=<-4.486 0 0>, rotation=<rot 90 0 0>}]
 dyn_R_armfur_F_jnt [{shapeType=Capsule, length=5, radius=2, shapeName=dyn_R_armfur_F_jnt_capsule, center=<-4.271 0 0>, rotation=<rot 90 0 0>}]
 dyn_R_armfur_G_jnt [{shapeType=Capsule, length=5, radius=2, shapeName=dyn_R_armfur_G_jnt_capsule, center=<-4.5147 0 0>, rotation=<rot 90 0 0>}]
 dyn_star_medallion_jnt [{
   shapeType=Capsule
   length=12.9357
   radius=11.613
   shapeName=dyn_star_medallion_jnt_capsule
   center=<13.0968 8.9818 -0.3542>
   rotation=<rot 0 -97.9606 -88.5327>
  }]
 lowerarm_l [{shapeType=Capsule, length=15, radius=3, shapeName=lowerarm_l_capsule, center=<9.5449 0 0>, rotation=<rot 90 0 0>}]
 lowerarm_r [{shapeType=Capsule, length=15, radius=3, shapeName=lowerarm_l_capsule, center=<-9.4981 0 0>, rotation=<rot 90 0 0>}]
 master_L_armfur_jnt [{shapeType=Capsule, length=0.505, radius=0.505, shapeName=master_L_armfur_jnt_capsule, rotation=<rot 0 -90 0>}]
 master_R_armfur_jnt [{shapeType=Capsule, length=0.505, radius=0.505, shapeName=master_R_armfur_jnt_capsule, rotation=<rot 0 -90 0>}]
 master_star_medallion_jnt [{shapeType=Capsule, length=0.505, radius=0.505, shapeName=master_star_medallion_jnt_capsule, center=<-0.1695 0 0>, rotation=<rot 0 -90 0>}]
 spine_02 [{shapeType=Capsule, length=10, radius=10, shapeName=spine_02_capsule, center=<0 3.4217 0>}]
 spine_03 [{shapeType=Capsule, length=10, radius=10, shapeName=spine_03_capsule, center=<0.9712 1.492 0>}]
 spine_04 [{shapeType=Capsule, length=10, radius=10, shapeName=spine_04_capsule, center=<-0.4931 -0.891 0>}]
 spine_05 [{shapeType=Capsule, length=11, radius=9, shapeName=spine_05_capsule, center=<-2.906 -1.1027 0>}]
}
constraints [
 {bone1Name=dyn_L_armfur_A_jnt, bone2Name=master_L_armfur_jnt, swing1LimitDegrees=20, swing2LimitDegrees=20, twistLimitDegrees=2}
 {bone1Name=dyn_L_armfur_B_jnt, bone2Name=master_L_armfur_jnt, swing1LimitDegrees=20, swing2LimitDegrees=20, twistLimitDegrees=2}
 {bone1Name=dyn_L_armfur_C_jnt, bone2Name=master_L_armfur_jnt, swing1LimitDegrees=20, swing2LimitDegrees=20, twistLimitDegrees=2}
 {bone1Name=dyn_L_armfur_D_jnt, bone2Name=master_L_armfur_jnt, swing1LimitDegrees=20, swing2LimitDegrees=20, twistLimitDegrees=2}
 {bone1Name=dyn_L_armfur_E_jnt, bone2Name=master_L_armfur_jnt, swing1LimitDegrees=20, swing2LimitDegrees=20, twistLimitDegrees=2}
 {bone1Name=dyn_L_armfur_F_jnt, bone2Name=master_L_armfur_jnt, swing1LimitDegrees=20, swing2LimitDegrees=20, twistLimitDegrees=2}
 {bone1Name=dyn_L_armfur_G_jnt, bone2Name=master_L_armfur_jnt, swing1LimitDegrees=20, swing2LimitDegrees=20, twistLimitDegrees=2}
 {
  bone1Name=master_L_armfur_jnt
  bone2Name=lowerarm_l
  swing1LimitDegrees=45
  swing1Motion=Free
  swing2LimitDegrees=45
  swing2Motion=Free
  twistLimitDegrees=45
  twistMotion=Free
 }
 {
  bone1Name=master_R_armfur_jnt
  bone2Name=lowerarm_r
  swing1LimitDegrees=45
  swing1Motion=Free
  swing2LimitDegrees=45
  swing2Motion=Free
  twistLimitDegrees=45
  twistMotion=Free
 }
 {bone1Name=dyn_R_armfur_A_jnt, bone2Name=master_R_armfur_jnt, swing1LimitDegrees=20, swing2LimitDegrees=20, twistLimitDegrees=2}
 {bone1Name=dyn_R_armfur_B_jnt, bone2Name=master_R_armfur_jnt, swing1LimitDegrees=20, swing2LimitDegrees=20, twistLimitDegrees=2}
 {bone1Name=dyn_R_armfur_C_jnt, bone2Name=master_R_armfur_jnt, swing1LimitDegrees=20, swing2LimitDegrees=20, twistLimitDegrees=2}
 {bone1Name=dyn_R_armfur_D_jnt, bone2Name=master_R_armfur_jnt, swing1LimitDegrees=20, swing2LimitDegrees=20, twistLimitDegrees=2}
 {bone1Name=dyn_R_armfur_E_jnt, bone2Name=master_R_armfur_jnt, swing1LimitDegrees=20, swing2LimitDegrees=20, twistLimitDegrees=2}
 {bone1Name=dyn_R_armfur_F_jnt, bone2Name=master_R_armfur_jnt, swing1LimitDegrees=20, swing2LimitDegrees=20, twistLimitDegrees=2}
 {bone1Name=dyn_R_armfur_G_jnt, bone2Name=master_R_armfur_jnt, swing1LimitDegrees=20, swing2LimitDegrees=20, twistLimitDegrees=2}
 {bone1Name=dyn_star_medallion_jnt, bone2Name=master_star_medallion_jnt, swing1LimitDegrees=2, swing2LimitDegrees=90, twistLimitDegrees=2}
 {
  bone1Name=lowerarm_l
  bone2Name=spine_05
  swing1LimitDegrees=45
  swing1Motion=Free
  swing2LimitDegrees=45
  swing2Motion=Free
  twistLimitDegrees=45
  twistMotion=Free
 }
 {
  bone1Name=lowerarm_r
  bone2Name=spine_05
  swing1LimitDegrees=45
  swing1Motion=Free
  swing2LimitDegrees=45
  swing2Motion=Free
  twistLimitDegrees=45
  twistMotion=Free
 }
 {
  bone1Name=master_star_medallion_jnt
  bone2Name=spine_05
  swing1LimitDegrees=45
  swing1Motion=Free
  swing2LimitDegrees=45
  swing2Motion=Free
  twistLimitDegrees=45
  twistMotion=Free
 }
 {
  bone1Name=spine_05
  bone2Name=spine_04
  swing1LimitDegrees=45
  swing1Motion=Free
  swing2LimitDegrees=45
  swing2Motion=Free
  twistLimitDegrees=45
  twistMotion=Free
 }
 {
  bone1Name=spine_04
  bone2Name=spine_03
  swing1LimitDegrees=45
  swing1Motion=Free
  swing2LimitDegrees=45
  swing2Motion=Free
  twistLimitDegrees=45
  twistMotion=Free
 }
 {
  bone1Name=spine_03
  bone2Name=spine_02
  swing1LimitDegrees=45
  swing1Motion=Free
  swing2LimitDegrees=45
  swing2Motion=Free
  twistLimitDegrees=45
  twistMotion=Free
 }
]
