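# Physics asset for the SunStar medium male face accessory, giving capsule bodies and joint limits to its dynamic ear-wing, face-tuft and hair bones.
# Each ear wing and the hair form two-link chains rooted on a small head capsule, while large and small tufts on each side attach directly to the head.
# Hair and ear wings have tight angular limits, while tufts swing more widely with very little twist.
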PhysicsAsset /BRCosmetics/Characters/Player/Male/Medium/Bodies/M_MED_SunStar/Meshes/Parts/M_MED_SunStar_FaceAcc_Physics
bodies {
 dyn_L_earwing_1_jnt [{shapeType=Capsule, length=5, radius=5, shapeName=dyn_L_earwing_1_jnt_capsule, center=<4.3532 0 0>}]
 dyn_L_earwing_2_jnt [{shapeType=Capsule, length=5, radius=5, shapeName=dyn_L_earwing_2_jnt_capsule, center=<5.3252 0 0>}]
 dyn_L_lrg_face_tuft_jnt [{
   shapeType=Capsule
   length=3
   radius=3.75
   shapeName=dyn_L_lrg_face_tuft_jnt_capsule
   center=<5.2355 0 -0.5152>
   rotation=<rot 80 -0.0025 -0.0026>
  }]
 dyn_L_sml_face_tuft_jnt [{
   shapeType=Capsule
   length=3.5
   radius=3.75
   shapeName=dyn_L_sml_face_tuft_jnt_capsule
   center=<0.1917 0 -2.008>
   rotation=<rot -50 0.001 -0.0011>
  }]
 dyn_R_earwing_1_jnt [{shapeType=Capsule, length=5, radius=5, shapeName=dyn_R_earwing_1_jnt_capsule, center=<-4.7746 0 0>}]
 dyn_R_earwing_2_jnt [{shapeType=Capsule, length=5, radius=5, shapeName=dyn_R_earwing_2_jnt_capsule, center=<-4.9364 0 0>}]
 dyn_R_lrg_face_tuft_jnt [{
   shapeType=Capsule
   length=3
   radius=3.75
   shapeName=dyn_L_lrg_face_tuft_jnt_capsule
   center=<-5.0085 0 0.0083>
   rotation=<rot 80 -0.0025 -0.0026>
  }]
 dyn_R_sml_face_tuft_jnt [{
   shapeType=Capsule
   length=3.5
   radius=3.75
   shapeName=dyn_L_sml_face_tuft_jnt_capsule
   center=<-0.5573 0 2.2667>
   rotation=<rot -50 0.001 -0.0011>
  }]
 dyn_hair_1_jnt [{
   shapeType=Capsule
   length=5.75
   radius=3.25
   shapeName=dyn_hair_1_jnt_capsule
   center=<2.6806 0 -0.4727>
   rotation=<rot 10 179.9994 179.9999>
  }]
 dyn_hair_2_jnt [{shapeType=Capsule, length=5, radius=3.5, shapeName=dyn_hair_2_jnt_capsule, center=<5.9261 0 -1.0449>, rotation=<rot 80 -0.0002 0.0002>}]
 master_head_jnt [{shapeType=Capsule, length=1, radius=1, shapeName=master_head_jnt_capsule}]
}
constraints [
 {bone1Name=dyn_L_lrg_face_tuft_jnt, bone2Name=master_head_jnt, swing1LimitDegrees=45, swing2LimitDegrees=15, twistLimitDegrees=2}
 {bone1Name=dyn_L_sml_face_tuft_jnt, bone2Name=master_head_jnt, swing1LimitDegrees=45, swing2LimitDegrees=45, twistLimitDegrees=2}
 {bone1Name=dyn_hair_1_jnt, bone2Name=master_head_jnt, swing1LimitDegrees=8, swing2LimitDegrees=2, twistLimitDegrees=2}
 {bone1Name=dyn_hair_2_jnt, bone2Name=dyn_hair_1_jnt, swing1LimitDegrees=8, swing2LimitDegrees=2, twistLimitDegrees=2}
 {bone1Name=dyn_L_earwing_1_jnt, bone2Name=master_head_jnt, swing1LimitDegrees=8, swing2LimitDegrees=8, twistLimitDegrees=8}
 {bone1Name=dyn_L_earwing_2_jnt, bone2Name=dyn_L_earwing_1_jnt, swing1LimitDegrees=15, swing2LimitDegrees=12, twistLimitDegrees=6}
 {bone1Name=dyn_R_lrg_face_tuft_jnt, bone2Name=master_head_jnt, swing1LimitDegrees=45, swing2LimitDegrees=15, twistLimitDegrees=2}
 {bone1Name=dyn_R_sml_face_tuft_jnt, bone2Name=master_head_jnt, swing1LimitDegrees=45, swing2LimitDegrees=45, twistLimitDegrees=2}
 {bone1Name=dyn_R_earwing_1_jnt, bone2Name=master_head_jnt, swing1LimitDegrees=8, swing2LimitDegrees=8, twistLimitDegrees=8}
 {bone1Name=dyn_R_earwing_2_jnt, bone2Name=dyn_R_earwing_1_jnt, swing1LimitDegrees=15, swing2LimitDegrees=12, twistLimitDegrees=6}
]
