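# Animation Blueprint for the SunStar male face accessory, whose output pose is just the pose and curves copied from the attached parent mesh.
# A head-space RigidBody node feeds a sunglasses bone offset gated by GlassesUp, but that chain is not wired to the output.
# Update and customization-loaded events set GlassesUp from the Stage2 variant tag.
# An emote-based smile-mouth material switch has no event driving it.
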AnimBlueprint /BRCosmetics/Characters/Player/Male/Medium/Bodies/M_MED_SunStar/Meshes/Parts/M_MED_SunStar_FaceAcc_AnimBP
target_skeleton=&/M_MED_SunStar_FaceAcc_Skeleton
graphs [
 AnimGraph {nodes [
   n1 = RigidBody(ComponentPose=Local To Component()) {
    bClampLinearTranslationLimitToRefPose=true
    bOverrideWorldGravity=true
    overridePhysicsAsset=&/M_MED_SunStar_FaceAcc_Physics
    simulationSpace=BaseBoneSpace
    alphaBoolBlend {blendOption=Linear}
    baseBoneRef {boneName=head}
    componentAppliedLinearAccClamp=<3000 3000 3000>
    componentLinearAccScale=<0.3 0.2 0.2>
    componentLinearVelScale=<0.3 0.2 0.2>
    overrideWorldGravity {z=-980}
    simSpaceSettings {maxAngularAcceleration=9997, maxAngularVelocity=3, maxLinearAcceleration=250, maxLinearVelocity=500, velocityScaleZ=0, worldAlpha=1}
   }
   "Transform (Modify) Bone - Bone: dyn_sunglasses_jnt"!(
    ComponentPose=@n1
    bAlphaBoolEnabled:bool=Get GlassesUp() -> (GlassesUp:bool)
    Translation="0,1,0"
    Rotation="0,30,0"
    Scale="0.95,0.95,0.95"
   ) -> (Pose) {
    alphaInputType=Bool
    rotationMode=BMM_Additive
    rotationSpace=BCS_BoneSpace
    scaleMode=BMM_Replace
    scaleSpace=BCS_BoneSpace
    translationMode=BMM_Additive
    translationSpace=BCS_BoneSpace
    alphaBoolBlend {blendOption=Linear}
    boneToModify {boneName=dyn_sunglasses_jnt}
    rotation {yaw=30}
    scale=<0.95 0.95 0.95>
    translation {y=1}
   }
   n2 = Copy Pose From Mesh() {bCopyCurves=true, bUseAttachedParent=true}
   Output Pose(Result=Component To Local(ComponentPose=Local To Component(LocalPose=@n2))) {layerGroup=None}
  ]}
 EventGraph {nodes [
   n1 = Get Dynamic Material Instance(SkeletalMeshComponent=GetPartSkeletalMeshComponent())
   > n2 = Branch(Condition=Get bIsPlayingEmote())
   > SetScalarParameterValue(self=@n1, ParameterName=Smile Mouth animation Frame, Value=8)
   SetScalarParameterValue(execute=@n2.else, self=@n1, ParameterName=Smile Mouth animation Frame, Value=3)
   n3 = Event BlueprintUpdateAnimation()
   n4 = Branch(
    execute=[@n3.then, @n5.then]
    Condition=CharacterItemDefinitionHasAnyMetaTag(Tags=Make Array("[0]"=(TagName="Cosmetics.Variant.Property.Stage2")))
   )
   > Set GlassesUp(GlassesUp:bool=true) -> (Output_Get:bool)
   Set GlassesUp(execute=@n4.else, GlassesUp=false)
   n5 = Event OnCharacterCustomizationLoaded()
  ]}
]
variables [GlassesUp {type=boolean}]
